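# Sound cue for a distant town clock chime, an ambient one-shot spatialized with the tile ambient attenuation.
# A quality switch selects between two near-identical branches, each randomly picking one of four chime waves with equal weight.
# Chimes 01 and 02 are pitched well down and quietened, while 03 and 04 skip that step.
# Each branch then applies light modulation and a volume envelope that fades out between two and four seconds.
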SoundCue /Game/Sounds/Artemis/Biomes/Towns/Town_Spots/Spot_Town_Clock_Chime_Distant_Cue
root=QualityLevel_0
nodes [
 {
  id=Enveloper_2
  children [Modulator_2]
  properties {
   pitchCurve {editorCurveData {keys [{time=0}]}}
   volumeCurve {editorCurveData {keys [
      <at 0 1 RCIM_Cubic>
      {arriveTangent=-0.0158, interpMode=RCIM_Cubic, leaveTangent=-0.0158, tangentMode=RCTM_User, time=2, value=1}
      {arriveTangent=0, interpMode=RCIM_Cubic, leaveTangent=0, tangentMode=RCTM_User, time=4, value=0}
     ]}}
  }
 }
 {
  id=Enveloper_3
  children [Modulator_3]
  properties {
   pitchCurve {editorCurveData {keys [{time=0}]}}
   volumeCurve {editorCurveData {keys [
      <at 0 1 RCIM_Cubic>
      {arriveTangent=-0.0158, interpMode=RCIM_Cubic, leaveTangent=-0.0158, tangentMode=RCTM_User, time=2, value=1}
      {arriveTangent=0, interpMode=RCIM_Cubic, leaveTangent=0, tangentMode=RCTM_User, time=4, value=0}
     ]}}
  }
 }
 {id=Modulator_2, children [Random_1], properties {volumeMin=0.8}}
 {id=Modulator_3, children [Random_2]}
 {id=Modulator_6, children [WavePlayer_5], properties {pitchMax=0.6, pitchMin=0.5, volumeMax=0.5, volumeMin=0.3}}
 {id=Modulator_7, children [WavePlayer_6], properties {pitchMax=0.6, pitchMin=0.5, volumeMax=0.5, volumeMin=0.3}}
 {id=Modulator_8, children [WavePlayer_9], properties {pitchMax=0.6, pitchMin=0.5, volumeMax=0.5, volumeMin=0.3}}
 {id=Modulator_9, children [WavePlayer_10], properties {pitchMax=0.6, pitchMin=0.5, volumeMax=0.5, volumeMin=0.3}}
 {id=QualityLevel_0, children [Enveloper_2,Enveloper_2,Enveloper_3,Enveloper_3]}
 {id=Random_1, children [Modulator_6,Modulator_7,WavePlayer_7,WavePlayer_8], properties {weights [1,1,1,1]}}
 {id=Random_2, children [Modulator_8, Modulator_9, WavePlayer_11, WavePlayer_12], properties {weights [1,1,1,1]}}
 {id=WavePlayer_10, properties {soundWaveAssetPtr=&/Spot_Town_Clock_Chime_02}}
 {id=WavePlayer_11, properties {soundWaveAssetPtr=&/Spot_Town_Clock_Chime_03}}
 {id=WavePlayer_12, properties {soundWaveAssetPtr=&/Spot_Town_Clock_Chime_04}}
 {id=WavePlayer_5, properties {soundWaveAssetPtr=&/Spot_Town_Clock_Chime_01}}
 {id=WavePlayer_6, properties {soundWaveAssetPtr=&Spot_Town_Clock_Chime_02}}
 {id=WavePlayer_7, properties {soundWaveAssetPtr=&Spot_Town_Clock_Chime_03}}
 {id=WavePlayer_8, properties {soundWaveAssetPtr=&Spot_Town_Clock_Chime_04}}
 {id=WavePlayer_9, properties {soundWaveAssetPtr=&Spot_Town_Clock_Chime_01}}
]
properties {
 attenuationSettings=/Game/Sounds/Attenuation/Fort_TileAmbient_OneShot_Att
 bEnableBusSends=false
 maxDistance=8000
 soundClassObject=&Fort_Ambient_OneShots
 volumeMultiplier=1.504
 preEffectBusSends [{}]
}
